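; Gameplay ability that removes the Whiff disguise once the player runs out of Whiff items.
; Event filtering rejects weakspot hits and accepts damage events only for nonzero hits on player or wildlife pawns.
; For Whiff-related events, it compares the current item count to a threshold that depends on the instigator's Whiff tag.
; At or below it, it plays a GameplayCue, removes the granted disguise item and leaves the Whiff wildlife faction.
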(event Ability|EventActivateAbilityFromEvent
  (params (EventData :type Gameplay%20Event%20Data%20Structure%20%28by%20ref%29))
  (bind (_eventtag _instigator _target _optionalobject _optionalobject2 _contexthandle _instigatortags) (Utilities|Struct|BreakGameplayEventData EventData))
  (bind _returnvalue (GameplayTags|HasTag _instigatortags "(TagName=\"Item.Consumable.Whiff\")"))
  ;[1 1 1] Only care about events from a whiff item (if triggered), catches when thrown
  (if (or _returnvalue (Utilities|Equal(Object) _optionalobject "/PrimalGameplay/Items/Consumables/Whiff/WID_Athena_Whiff.WID_Athena_Whiff"))
    (Variables|Default|SetItemCountCheck (Math|Integer|SelectInt 1 0 _returnvalue))
    (CallFunction|TryRemoveBackpackItem)
    (else
      _))
  (Ability|EndAbility))

(event Ability|EventOnEndAbility
  (params (bWasCancelled :type Boolean)))

(event Custom|TryRemoveBackpackItem
  (bind _returnvalue (Ability|GetAvatarActorfromActorInfo))
  (bind _output_get (Variables|Default|SetOwningActor _returnvalue))
  (Utilities|IsValid _output_get
    (:Is%20Valid
      (bind _output_get_2 (Variables|Default|SetFPController (Fortnite|GetFortPlayerControllerfromActor _returnvalue)))
      ;[1 1 1] If we no longer have any "whiff items" on us AND we have the equipped item, remove it!
      (if (and (<= (Inventory|GetItemQuantityOnPlayer _output_get_2 (Variables|Default|GetWIDCurrentItem)) (Variables|Default|GetItemCountCheck)) (> (Inventory|GetItemQuantityOnPlayer _output_get_2 (Variables|Default|GetAGIDToRemove)) 0))
        ;[1 1 1] Apply GameplayCue
        (Ability|ExecuteGameplayCueOnOwner (Variables|Default|GetGC))
        (Networking|ForceNetUpdate (Variables|Default|GetOwningActor))
        ;[1 1 1] Remove the Character Part
        (Inventory|RemoveItemFromPlayer :PlayerController (Variables|Default|GetFPController) :ItemDefinition (Variables|Default|GetAGIDToRemove) :AmountToRemove -1 :bForceRemoval true :bSkipVerbRemoval false)
        (Affiliation|RemovefromFaction (Variables|Default|GetOwningActor) "(TagName=\"Athena.Faction.Wildlife.Special.Whiff\")")))))

(fn ShouldAbilityRespondToEvent
  (params (ActorInfo :type Gameplay%20Ability%20Actor%20Info%20Structure) (Payload :type Gameplay%20Event%20Data%20Structure))
  (bind (_eventtag _instigator _target _optionalobject _optionalobject2 _contexthandle _instigatortags _targettags _eventmagnitude) (Utilities|Struct|BreakGameplayEventData Payload))
  ;[1 1 1] If the event was the damage event check if it is a player pawn
  (switch Utilities|FlowControl|Switch|SwitchonGameplayTag _eventtag
    (:Default
      (return true))
    (:Event.Damage.Dealt.Weakspot
      (return false))
    (:Event.Damage.Dealt
      (if (== _eventmagnitude 0.0)
        (return false)
        (else
          (Utilities|Casting|CastToFortPlayerPawn _target
            (:then
              (return true))
            (:CastFailed
              (Utilities|Casting|CastToNPC_Pawn_Wildlife_Parent _target
                (:then
                  (return true))
                (:CastFailed
                  (return false))))))))))
