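; Primary gameplay ability for the Whiff consumable that disguises the user and adds them to the Whiff special wildlife faction.
; Activation defers to the parent, and the post-commit montage success event caches the pawn, gives the Whiff disguise item to its controller's inventory and applies the faction tag.
; When the ability ends, cancelled or not, it swaps quickbar focus back to the pickaxe if the cached pawn is valid.
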(event Ability|EventActivateAbility
  (|CallParentFunction))

(event EventOnMontageTriggeredPostCommitSuccess
  ;[1 1 1] On Montage Succeeded
  (bind _asfort_pawn (Utilities|Casting|CastToFortPawn (Ability|GetAvatarActorfromActorInfo)))
  (bind _output_get (Variables|Default|SetPawn _asfort_pawn))
  (Inventory|GiveItemtoInventoryOwner (Fortnite|GetFortPlayerControllerfromActor _output_get) "/PrimalGameplay/Items/Consumables/Whiff/AGID_Disguise_Whiff.AGID_Disguise_Whiff")
  (Affiliation|AddtoFaction (Variables|Default|GetPawn) "(TagName=\"Athena.Faction.Wildlife.Special.Whiff\")"))

(event Ability|EventOnEndAbility
  (params (bWasCancelled :type Boolean))
  (bind _pawn (Variables|Default|GetPawn))
  ;[1 1 1] Swap to pickaxe on ability end
  (|CallParentFunction bWasCancelled)
  (if (Utilities|IsValid _pawn :!class /Script/Engine.KismetSystemLibrary)
    (Quickbars|ClientSwapQuickBarFocus (Fortnite|GetFortPlayerControllerfromActor _pawn) "Primary" 0 false)))
